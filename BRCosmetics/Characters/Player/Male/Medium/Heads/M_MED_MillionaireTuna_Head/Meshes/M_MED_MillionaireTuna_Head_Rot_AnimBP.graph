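# Animation Blueprint for the MillionaireTuna head cosmetic, copying the owner mesh pose and layering an idle over head, fins, tail and eyes, weakened on the head during emotes.
# A pose asset, head scaling, light spring dynamics and a spine-to-head rotation-driven offset follow, then get cached.
# Output shrinks neck_01 to near zero when the playlist uses custom character parts.
# Neck counter-rotations are computed, but their bone nodes don't reach the output.
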AnimBlueprint /BRCosmetics/Characters/Player/Male/Medium/Heads/M_MED_MillionaireTuna_Head/Meshes/M_MED_MillionaireTuna_Head_Rot_AnimBP
target_skeleton=/Game/Characters/Player/Male/Medium/Heads/M_MED_MillionaireTuna_Head/Meshes/M_MED_MillionaireTuna_Head_Skeleton
graphs [
 AnimGraph {nodes [
   Component To Local()
   n1 = "Transform (Modify) Bone - Bone: neck_01"(Alpha=1, Translation="0,0,0", Rotation=Get neck rot_2() -> (neck rot_2:Rotator), Scale="1,1,1") {
    rotationMode=BMM_Additive
    rotationSpace=BCS_BoneSpace
    translationMode=BMM_Additive
    translationSpace=BCS_BoneSpace
    alphaBoolBlend {blendOption=Linear}
    alphaScaleBiasClamp {interpSpeedDecreasing=50, interpSpeedIncreasing=50}
    boneToModify {boneName=neck_01}
    rotation {yaw=-90}
   }
   n2 = "Transform (Modify) Bone - Bone: neck_02"(ComponentPose=@n1, Alpha=1, Translation="0,0,0", Rotation=Get neck rot() -> (neck rot:Rotator), Scale="1,1,1") {
    rotationMode=BMM_Additive
    rotationSpace=BCS_BoneSpace
    translationMode=BMM_Additive
    translationSpace=BCS_BoneSpace
    alphaBoolBlend {blendOption=Linear}
    alphaScaleBiasClamp {interpSpeedDecreasing=50, interpSpeedIncreasing=50}
    boneToModify {boneName=neck_02}
    rotation {yaw=-90}
   }
   Component To Local(ComponentPose=@n2)
   n3 = "Transform (Modify) Bone - Bone: neck_01"!(ComponentPose=Local To Component(LocalPose=Use cached pose 'PreScaleCache'()), Alpha=1) -> (Pose) {scaleMode=BMM_Replace, alphaBoolBlend {blendOption=Linear}, boneToModify {boneName=neck_01}, scale=<0.01 0.01 0.01>}
   Output Pose(Result=Blend Poses by bool(
     BlendPose_0=Component To Local(ComponentPose=@n3)
     BlendPose_1=Use cached pose 'PreScaleCache'()
     BlendTime_0=0
     BlendTime_1=0
     bActiveValue=Get GetPlayerListUsesCustomCharacterParts() -> (GetPlayerListUsesCustomCharacterParts:bool)
    ) {blendType=Linear}) {layerGroup=None}
   n4 = Layered blend per bone!(
    BasePose=Copy Pose From Mesh(SourceMeshComponent=Get MeshToCopy() -> (MeshToCopy:Skeletal Mesh Component Object Reference)) {bCopyCurves=true}
    BlendPoses_0=Sequence Player 'MillionaireTuna_Head_idle'() {sequence=&/MillionaireTuna_Head_idle}
    BlendPoses_1:Pose Link Structure=Sequence Player 'MillionaireTuna_Head_idle'() {sequence=&MillionaireTuna_Head_idle}
    BlendWeights_0=Get blend_emote_alpha() -> (blend_emote_alpha:double)
    BlendWeights_1:float=1
   ) -> (Pose) {
    blendPoses [{linkId=-1, sourceLinkId=-1}, {linkId=-1, sourceLinkId=-1}]
    blendWeights [1,1]
    layerSetup [
     {branchFilters [{blendDepth=1, boneName=head}]}
     {branchFilters [
       {blendDepth=1, boneName=fin_l}
       {blendDepth=1, boneName=fin_r}
       {blendDepth=1, boneName=fin_c}
       {blendDepth=1, boneName=Tail}
       {blendDepth=1, boneName=L_Eye}
       {blendDepth=1, boneName=R_Eye}
      ]}
    ]
   }
   n5 = M_MED_MillionaireTuna_Head_PoseAsset!(SourcePose:Pose Link Structure=@n4) -> (Pose:Pose Link Structure) {
    is=title AnimGraphNode_PoseBlendNode
    blendOption=Linear
    poseAsset=&/M_MED_MillionaireTuna_Head_PoseAsset
   }
   n6 = "Transform (Modify) Bone - Bone: head"!(ComponentPose=Local To Component(LocalPose=@n5), Alpha=1) -> (Pose) {
    scaleMode=BMM_Replace
    scaleSpace=BCS_BoneSpace
    translationMode=BMM_Additive
    translationSpace=BCS_BoneSpace
    alphaBoolBlend {blendOption=Linear}
    boneToModify {boneName=head}
    scale=<0.925 0.925 0.925>
    translation {y=1}
   }
   n7 = Anim Dynamics - Bone: head(ComponentPose=@n6, Alpha=0.125, ExternalForce="0,0,0") {
    bAngularSpring=true
    bGravityOverrideInSimSpace=true
    bLinearSpring=true
    gravityScale=0.25
    linearDampingOverride=10
    linearSpringConstant=15
    alphaBoolBlend {blendOption=Linear}
    boundBone {boneName=head}
    physicsBodyDefinitions [{
      sphereCollisionRadius=0
      boundBone {boneName=head}
      boxExtents=<120 60 60>
      constraintSetup {angularLimitsMax=<0 0 0>, angularLimitsMin=<0 0 0>, angularTarget=<0 0 0>, linearAxesMax=<0 0 9>, linearAxesMin=<0 0 -9>}
      localJointOffset=<0 0 0>
     }]
   }
   n8 = "Transform (Modify) Bone - Bone: head"!(ComponentPose=@n7, Alpha=Get HeadRot() -> (HeadRot:double), Translation:Vector="0,-2.5,0") -> (Pose) {
    scaleMode=BMM_Additive
    scaleSpace=BCS_BoneSpace
    translationMode=BMM_Additive
    translationSpace=BCS_BoneSpace
    alphaBoolBlend {blendOption=Linear}
    boneToModify {boneName=head}
    scale=<0.925 0.925 0.925>
    translation {y=-2.5}
   }
   Save cached pose 'PreScaleCache'(Pose=Component To Local(ComponentPose=@n8)) {cacheName=PreScaleCache}
  ]}
 EventGraph {nodes [
   Event BlueprintUpdateAnimation()
   > n1 = Sequence()
   n2 = GetOwnerSkeletalMeshComponent()
   n3 = BreakRotator(InRot=BreakTransform(InTransform=GetDeltaTransformFromRefPose(self=@n2, BoneName=neck_01, BaseName=spine_05)).Rotation)
   > Set neck rot_2(neck rot_2:Rotator=MakeRotator(Roll=float * float!(A=@n3, B=-1, C:double) -> (ReturnValue), Pitch=float * float(A=@n3.Pitch, B=-1), Yaw=float * float(A=@n3.Yaw, B=-1))) -> (Output_Get:Rotator)
   n4 = BreakRotator(InRot=BreakTransform(InTransform=GetDeltaTransformFromRefPose(self=@n2, BoneName=neck_02, BaseName=neck_01)).Rotation)
   > Set neck rot(neck rot:Rotator=MakeRotator(Roll=float * float(A=@n4, B=-1, C=0.35), Pitch=float * float(A=@n4.Pitch, B=-1, C=0.35), Yaw=float * float(A=@n4.Yaw, B=-1, C=0.35))) -> (Output_Get:Rotator)
   n5 = BreakRotator(InRot=BreakTransform(InTransform=GetDeltaTransformFromRefPose(self=@n2, BoneName=head, BaseName=neck_02)).Rotation)
   > Set head rot(head rot:Rotator=MakeRotator(Roll=float * float(A=@n5, B=-1, C=0.1), Pitch=float * float(A=@n5.Pitch, B=-1, C=0.1), Yaw=float * float(A=@n5.Yaw, B=-1, C=0.1))) -> (Output_Get:Rotator)
   > n6 = Branch(Condition=Get bIsPlayingEmote())
   > Set blend_emote_alpha(blend_emote_alpha:double=0.3) -> (Output_Get:double)
   Set blend_emote_alpha(execute=@n6.else, blend_emote_alpha=1)
   Set HeadRot(
    execute=@n1.then_1
    HeadRot:double=GetAlphaFromReferenceTransformDelta(SkeletalMeshComponent=Get Skel Mesh(), BoneName=spine_05, BaseName=head, InputRange=(X=0,Y=-25), TwistAxis="0,0,1")
   ) -> (Output_Get:double)
   Event BlueprintInitializeAnimation()
   > n7 = Sequence()
   > Set MeshToCopy(MeshToCopy:Skeletal Mesh Component Object Reference=GetOwnerSkeletalMeshComponent()) -> (Output_Get:Skeletal Mesh Component Object Reference)
   > Set GetPlayerListUsesCustomCharacterParts(GetPlayerListUsesCustomCharacterParts:bool=GetPlaylistUsesCustomCharacterParts!() -> (ReturnValue:bool)) -> (Output_Get:bool)
   Set Skel Mesh(execute=@n7.then_1, Skel Mesh=GetOwnerSkeletalMeshComponent())
   TryGetPawnOwner()
  ]}
]
variables [
 blend_emote_alpha {type=number}
 GetPlayerListUsesCustomCharacterParts {type=boolean}
 head rot {type=Rotator}
 HeadDownRot {type=number}
 HeadRot {type=number}
 HeadUpRot {type=number}
 isCrouchingAlpha {type=number}
 isSkydivingAlpha {type=number}
 MeshToCopy {type=SkeletalMeshComponent}
 neck rot {type=Rotator}
 neck rot_2 {type=Rotator}
 Skel Mesh {type=SkeletalMeshComponent}
 translation {type=Vector}
]
